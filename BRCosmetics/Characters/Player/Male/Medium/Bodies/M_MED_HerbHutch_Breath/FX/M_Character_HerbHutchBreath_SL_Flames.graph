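# Translucent unlit Niagara mesh-particle material for stylized green flames on the HerbHutch Breath cosmetic character.
# A distorted, panning trail texture and UV masks form flame shape and glow, with erosion reveal plus depth and camera fades.
# Noise and vertical masks blend emissive colors, adding a bright inner tint.
# WPO combines SplineThicken, sine wiggle, clamped position drag and bend, with dynamic parameters driving width, motion, offset and reveal.
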Material /BRCosmetics/Characters/Player/Male/Medium/Bodies/M_MED_HerbHutch_Breath/FX/M_Character_HerbHutchBreath_SL_Flames
graph {
 properties {
  bAutomaticallySetUsageInEditor=false
  bUsedWithNiagaraMeshParticles=true
  blendMode=BLEND_Translucent
  shadingModel=MSM_Unlit
  usedShadingModels=Unlit
 }
 nodes [
  Constant() {r=10}
  n1 = VectorParameter() {parameterName=BendDirection, defaultValue=<rgba 1 0 0 1>}
  MaterialFunctionCall(X=@n1.R, Y=@n1.R, Z=@n1.R) {materialFunction=&MakeFloat3}
  n2 = DynamicParameter() {defaultValue=<rgba 1 1 0 1>, paramNames [WidthTipMultiplier, WidthBaseMultiplier, Ribbon Texture Offset, Reveal]}
  n3 = Multiply(A=ScalarParameter() {defaultValue=7, group=Spline, parameterName=Width Base}, B=@n2.WidthBaseMultiplier)
  n4 = Multiply(A=@n2.WidthTipMultiplier, B=ScalarParameter() {defaultValue=7, group=Spline, parameterName=Width Tip})
  n5 = StaticBool() {value=true}
  n6 = ComponentMask(None=TextureCoordinate()) {g=true}
  n7 = DynamicParameter() {defaultValue=<rgba 1 1 0 1>, paramNames [WidthTipMultiplier, WidthBaseMultiplier, Ribbon Texture Offset, Reveal]}
  n8 = Multiply(A=@n7.Ribbon Texture Offset, B=Constant() {r=5})
  n9 = Constant() {r=0.4}
  n10 = Multiply(A=Add(A=Add(A=@n6, B=@n8), B=@n9), B=ScalarParameter() {defaultValue=0.6, group=WPO, parameterName=WPO Frequency, sliderMax=2})
  n11 = Multiply(A=Time() {period=1}, B=Multiply(A=@n9, B=ScalarParameter() {defaultValue=1, group=WPO, parameterName=WPO Speed, sliderMax=1}))
  n12 = Sine(None=Add(A=Add(A=@n10, B=@n11), B=ParticleRandom()))
  n13 = Multiply(A=@n12, B=Transform(None=VectorParameter() {group=WPO, parameterName=WPO Tangent Vector, defaultValue=<rgba 1 0 1 1>}.RGB))
  n14 = DynamicParameter() {parameterIndex=1, defaultValue=<rgba 0 0 0 0>, paramNames [Bend Amount, Bend Up/Down Bias, WPO Wiggle, ""]}
  n15 = Multiply(A=ConstantBiasScale(None=@n12) {bias=0.8}, B=Constant3Vector() {constant=<rgba 0 0 1 0>})
  n16 = ScalarParameter() {defaultValue=1, group=WPO, parameterName=WPO Upwards, sliderMax=50}
  n17 = Multiply(
   A=Add(A=Multiply(A=@n13, B=@n14.WPO Wiggle), B=Multiply(A=@n15, B=@n16))
   B=ScalarParameter() {defaultValue=1, group=WPO, parameterName=WPO Wave, sliderMax=10, sortPriority=33}
  )
  n18 = DynamicParameter() {parameterIndex=2, defaultValue=<rgba 0 0 0 0>, paramNames [Interpolated Pos X, Interpolated Pos Y, Interpolated Pos Z, ""]}
  n19 = MaterialFunctionCall(X=@n18.Interpolated Pos X, Y=@n18.Interpolated Pos X, Z=@n18.Interpolated Pos X) {materialFunction=&MakeFloat3}
  n20 = ParticlePositionWS()
  n21 = ScalarParameter() {defaultValue=20, group=WPO, parameterName=WPO Max Drag, sliderMax=50}
  n22 = Multiply(
   A=VectorParameter() {group=WPO, parameterName=WPO Drag Direction Mask, defaultValue=<rgba 0.1 0.1 0.1 0.1>}.RGB
   B=Clamp(None=Subtract(A=@n19.Result, B=@n20), Min=Multiply(A=@n21) {constB=-1}, Max=@n21)
  )
  n23 = OneMinus(None=ComponentMask(None=TextureCoordinate()) {g=true})
  n24 = ScalarParameter() {defaultValue=1.5, group=WPO, parameterName=WPO Bend, sliderMax=10}
  n25 = Multiply(
   A=ScalarParameter() {defaultValue=1, group=WPO, parameterName=WPO Intensity, sliderMax=100}
   B=LinearInterpolate(B=Add(A=@n17, B=@n22), Alpha=Power(Base=@n23, Exp=@n24))
  )
  n26 = LinearInterpolate(
   A=Constant3Vector() {constant=<rgba 1 0 0 0>}
   B=Constant3Vector() {constant=<rgba 0 0 1 0>}
   Alpha=ScalarParameter() {parameterName=X or Z Axis Direction}
  )
  n27 = DynamicParameter() {parameterIndex=1, defaultValue=<rgba 0 0 0 0>, paramNames [Bend Amount, Bend Up/Down Bias, WPO Wiggle, ""]}
  n28 = ComponentMask(None=TextureCoordinate()) {g=true}
  n29 = OneMinus(None=@n28)
  n30 = Constant() {r=3}
  n31 = Power(Base=@n28, Exp=@n30)
  n32 = Multiply(
   A=Multiply(A=Transform(None=@n26) {transformSourceType=TRANSFORMSOURCE_Local}, B=@n27.Bend Amount)
   B=LinearInterpolate(A=Power(Base=@n29, Exp=@n30), B=OneMinus(None=@n31), Alpha=@n27.Bend Up/Down Bias) {constAlpha=0}
  )
  NamedRerouteDeclaration(None=Add(
    A=MaterialFunctionCall(WidthBase=@n3, WidthTip=@n4, Expand U or V UV Channel=@n5) {materialFunction=&SplineThicken}.WorldPositionOffset
    B=Add(A=@n25, B=@n32)
   )) {name=WPO, nodeColor=<rgba 0.5703 0 1 1>}
  n33 = Constant() {r=0.6}
  n34 = Constant() {r=1}
  n35 = Add(
   A=MaterialFunctionCall(
    UVs=TextureCoordinate()
    Texture Scale=MaterialFunctionCall(X=@n33, Y=@n34) {materialFunction=&MakeFloat2}.Result
   ) {materialFunction=&ScaleUVsByCenter}.UVs
   B=MaterialFunctionCall(X=Constant(), Y=ScalarParameter() {defaultValue=-0.35}) {materialFunction=&MakeFloat2}.Result
  )
  n36 = Multiply(A=NamedRerouteUsage(), B=Constant() {r=0.1})
  n37 = Power(
   Base=SphereMask(A=Add(A=@n35, B=@n36), B=Constant() {r=0.5}, Radius=Constant() {r=0.3}, Hardness=Constant()) {hardnessPercent=0}
   Exp=Constant() {r=3}
  )
  n38 = Constant() {r=0.2}
  n39 = Add(A=@n38, B=Constant() {r=1})
  n40 = Add(A=TextureCoordinate(), B=ParticleRandom())
  n41 = MaterialFunctionCall(X=Constant() {r=0.5}, Y=Constant() {r=0.5}) {materialFunction=&MakeFloat2}
  n42 = Constant() {r=0.1}
  n43 = Multiply(A=Constant() {r=1}, B=NamedRerouteUsage())
  n44 = TextureSampleParameter2D(UVs=Panner(
    Coordinate=Multiply(A=@n40, B=@n41.Result)
    Speed=MaterialFunctionCall(X=@n42, Y=@n43) {materialFunction=&MakeFloat2}.Result
   )) {
   group=03- Color
   parameterName=ColorTex
   samplerType=SAMPLERTYPE_LinearColor
   texture=/Game/Effects/Fort_Effects/Gameplay/SmokeDustExplosions/Textures/T_NoisePack_LinearColor
  }
  n45 = Multiply(A=MaterialFunctionCall(Float2=TextureCoordinate()) {materialFunction=&BreakOutFloat2Components}.G)
  n46 = Constant() {r=20}
  n47 = Saturate(None=Add(A=SmoothStep(Min=@n38, Max=@n39, Value=@n44.R) {constMin=0.5}, B=Power(Base=@n45, Exp=@n46)) {constB=0})
  NamedRerouteDeclaration(None=LinearInterpolate(
    A=Constant() {r=1}
    B=VectorParameter() {parameterName=InnerAdditionalColor, defaultValue=<rgba 14.1433 20 0 1>}.RGB
    Alpha=Multiply(A=@n37, B=@n47)
   )) {name=InsideColorMultiplier, nodeColor=<rgba 1 0 0.75 1>}
  n48 = NamedRerouteUsage()
  n49 = NamedRerouteUsage()
  n50 = Multiply(
   A=SmoothStep(
    Min=ScalarParameter() {group=00- FlameShape, parameterName=Opacity_TextureMin}
    Max=ScalarParameter() {defaultValue=0.4, group=00- FlameShape, parameterName=Opacity_TextureMax}
    Value=Multiply(A=@n48, B=@n49)
   ) {constMax=0.4}
   B=ScalarParameter() {defaultValue=1, group=01-Opacity, parameterName=MainTrailOpacity}
  )
  n51 = Power(
   Base=Multiply(A=NamedRerouteUsage(), B=NamedRerouteUsage())
   Exp=ScalarParameter() {defaultValue=0.5, group=01-Opacity, parameterName=GlowPow}
  )
  n52 = ScalarParameter() {defaultValue=3, group=01-Opacity, parameterName=GlowStrength}
  n53 = Add(A=@n50, B=NamedRerouteDeclaration(None=Multiply(A=@n51, B=@n52)) {name=GlowMask, nodeColor=<rgba 1 0.5859 0 1>})
  n54 = ScalarParameter() {defaultValue=35, group=01-Opacity, parameterName=Depth Fade Distance, sliderMax=100}
  n55 = DepthFade(Opacity=@n53, FadeDistance=LinearInterpolate(A=@n54, B=Multiply(A=@n54) {constB=0.5}, Alpha=@n50)) {fadeDistanceDefault=25}
  n56 = Multiply(
   A=ShadingPathSwitch(Default=@n55, Deferred=@n55, Forward=@n53, Mobile=@n53)
   B=MaterialFunctionCall(
    Fade Length=ScalarParameter() {defaultValue=100, group=01-Opacity, parameterName=Camera Fade Length, sliderMax=200}
    Fade Offset=ScalarParameter() {defaultValue=25, group=01-Opacity, parameterName=Camera Fade Offset, sliderMax=200}
   ) {materialFunction=&CameraDepthFade}.Result
  )
  n57 = ScalarParameter() {defaultValue=0.4, group=01-Opacity, parameterName=FresnelExponent}
  n58 = ScalarParameter() {defaultValue=0.04, group=01-Opacity, parameterName=FresnelFraction}
  n59 = LinearInterpolate(B=OneMinus(None=Fresnel(ExponentIn=@n57, BaseReflectFractionIn=@n58)), Alpha=ScalarParameter() {group=01-Opacity, parameterName=UseFresnel}) {constA=1}
  n60 = Multiply(A=Multiply(A=Multiply(A=@n56, B=@n59), B=ParticleColor().A), B=NamedRerouteUsage())
  n61 = Multiply(
   A=MaterialFunctionCall(Float2=TextureCoordinate()) {materialFunction=&BreakOutFloat2Components}.G
   B=ScalarParameter() {defaultValue=1, parameterName=BottomFade}
  )
  n62 = ScalarParameter() {defaultValue=10, parameterName=BottomPower}
  n63 = Saturate(None=OneMinus(None=Power(Base=@n61, Exp=@n62)))
  NamedRerouteDeclaration(None=Saturate(None=Multiply(A=@n60, B=@n63))) {name=Opacity, nodeColor=<rgba 0 1 0.8359 1>}
  NamedRerouteDeclaration(None=ScalarParameter() {defaultValue=1, parameterName=FlameSpeedMult}) {name=MasterSpeed, nodeColor=<rgba 0 0.8125 1 1>}
  n64 = NamedRerouteUsage()
  n65 = OneMinus(None=MaterialFunctionCall(Float2=TextureCoordinate()) {materialFunction=&BreakOutFloat2Components}.G)
  n66 = DynamicParameter() {defaultValue=<rgba 1 1 0 1>, paramNames [WidthTipMultiplier, WidthBaseMultiplier, Ribbon Texture Offset, Reveal]}
  n67 = ScalarParameter() {defaultValue=0.2, group=02- Reveal, parameterName=Reveal Softness}
  NamedRerouteDeclaration(None=MaterialFunctionCall!(
    Texture=OneMinus(None=@n64)
    Erosion=MaterialFunctionCall(Texture=@n65, Erosion=@n66.Reveal, Softness=@n67) {materialFunction=/Game/Effects/Niagara/MasterMaterials/Functions/MF_Erosion_Smoothstep_01}.Result
    Softness=Constant() {r=0.5}
   ) {materialFunction=&MF_Erosion_Smoothstep_01}.Result) {name=RevealMask, nodeColor=<rgba 0.125 0 1 1>}
  n68 = TextureCoordinate()
  n69 = VectorParameter() {group=03- Color, parameterName=ColorTex_Settings, channelNames {a=PanY, b=PanX, g=ScaleY, r=ScaleX}, defaultValue=<rgba 1 1 0 0.4>}
  n70 = AppendVector(A=@n69.PanX, B=Multiply(A=@n69.PanY, B=NamedRerouteUsage()))
  n71 = DynamicParameter() {defaultValue=<rgba 1 1 0 1>, paramNames [WidthTipMultiplier, WidthBaseMultiplier, Ribbon Texture Offset, Reveal]}
  n72 = Add(
   A=Multiply(A=Panner(Coordinate=@n68, Speed=@n70) {speedX=-0.5}, B=AppendVector(A=@n69.ScaleX, B=@n69.ScaleX))
   B=AppendVector(A=Constant(), B=@n71.Ribbon Texture Offset)
  )
  n73 = ParticleRandom()
  n74 = NamedRerouteUsage()
  n75 = ScalarParameter() {defaultValue=0.2, group=03- Color, parameterName=ColorDistoStrength}
  n76 = TextureSampleParameter2D(UVs=Add(A=Add(A=@n72, B=@n73), B=Multiply(A=@n74, B=@n75))) {
   group=03- Color
   parameterName=ColorTex
   samplerType=SAMPLERTYPE_LinearColor
   texture=&T_NoisePack_LinearColor
  }
  NamedRerouteDeclaration(None=@n76.G) {name=ColorNoiseGreen, nodeColor=<rgba 0 1 0.6719 1>}
  n77 = DynamicParameter() {defaultValue=<rgba 1 1 0 1>, paramNames [WidthTipMultiplier, WidthBaseMultiplier, Ribbon Texture Offset, Reveal]}
  n78 = AppendVector(A=Constant(), B=Multiply(A=@n77.Ribbon Texture Offset, B=Constant() {r=2}) {constB=3})
  n79 = TextureCoordinate()
  n80 = VectorParameter() {group=06- Distortion, parameterName=DistoUV_Settings, channelNames {a=PanY, b=PanX, g=ScaleY, r=ScaleX}, defaultValue=<rgba 1 1 0 1>}
  n81 = Multiply(A=@n80.PanY, B=NamedRerouteUsage())
  n82 = Multiply(A=Panner(Coordinate=@n79, Speed=AppendVector(A=@n80.PanX, B=@n81)) {speedX=-0.5}, B=AppendVector(A=@n80.ScaleX, B=@n80.ScaleX))
  n83 = TextureSampleParameter2D(UVs=Add(A=Add(A=@n78, B=@n82), B=ParticleRandom())) {
   group=06- Distortion
   parameterName=Distortion Texture3
   samplerType=SAMPLERTYPE_LinearGrayscale
   texture=/Game/Effects/Fort_Effects/Textures/Tile/T_CellNoise_Tile_M
  }
  n84 = Add(A=ComponentMask(None=@n79) {g=true}) {constB=-0.5}
  NamedRerouteDeclaration(None=Multiply(A=@n83.R, B=Multiply(A=@n84) {constB=-1})) {name=Distortion, nodeColor=<rgba 1 0.3083 0.7353 1>}
  n85 = TextureCoordinate()
  n86 = NamedRerouteUsage()
  n87 = ScalarParameter() {defaultValue=0.05, group=06- Distortion, parameterName=Distortion Strength, sliderMax=1, sliderMin=-1}
  n88 = ScalarParameter() {defaultValue=1.5, group=00- FlameShape, parameterName=TrailSpeedX, sliderMax=5, sliderMin=-5}
  n89 = NamedRerouteUsage()
  n90 = Panner(
   Coordinate=Add(
    A=MaterialFunctionCall(XY=@n85) {materialFunction=&Swizzle}.YX
    B=Multiply(A=@n86, B=@n87) {constB=0.1}
   )
   Speed=AppendVector(A=Multiply(A=@n88, B=@n89), B=Constant())
  ) {speedX=1}
  n91 = Constant() {r=1}
  n92 = AppendVector(A=ScalarParameter() {defaultValue=0.5, group=00- FlameShape, parameterName=Scale X, sliderMax=5}, B=@n91)
  n93 = ScalarParameter() {defaultValue=2, group=00- FlameShape, parameterName=Scale Y, sliderMax=5}
  n94 = MaterialFunctionCall(Float2=MaterialFunctionCall(
    UVs=Multiply(A=@n90, B=@n92)
    Texture Scale=MaterialFunctionCall(X=@n91, Y=@n93) {materialFunction=&MakeFloat2}.Result
   ) {materialFunction=&ScaleUVsByCenter}.UVs) {materialFunction=&BreakOutFloat2Components}
  n95 = DynamicParameter() {defaultValue=<rgba 1 1 0 1>, paramNames [WidthTipMultiplier, WidthBaseMultiplier, Ribbon Texture Offset, Reveal]}
  n96 = Constant() {r=0.8}
  n97 = Add(
   A=MaterialFunctionCall(X=@n94.R, Y=Saturate(None=@n94.G)) {materialFunction=&MakeFloat2}.Result
   B=AppendVector(A=Multiply(A=@n95.Ribbon Texture Offset, B=@n96), B=Constant())
  )
  n98 = MaterialFunctionCall(XY=@n97) {materialFunction=&Swizzle}
  n99 = ScalarParameter() {group=00- FlameShape, parameterName=SwizzleTrail}
  n100 = ParticleRandom()
  n101 = Constant()
  n102 = TextureSample(UVs=Add(
    A=LinearInterpolate(A=@n97, B=@n98.YX, Alpha=@n99)
    B=MaterialFunctionCall(X=@n100, Y=@n101) {materialFunction=&MakeFloat2}.Result
   )) {texture=/BRCosmetics/Characters/Player/Female/Medium/Bodies/F_MED_MadameMoth_Posh/FX/FX_Rig_Pipeline/T_Rig_MadameMoth_Fire_Trail01}
  NamedRerouteDeclaration(None=@n102.R) {name=FlameTrail, nodeColor=<rgba 0 1 0.2734 1>}
  NamedRerouteDeclaration(None=@n83.R) {name=DistortionTexture, nodeColor=<rgba 0.125 0 1 1>}
  NamedRerouteDeclaration(None=@n102.G) {name=FlameTrailGlow, nodeColor=<rgba 1 0 0.8906 1>}
  n103 = TextureCoordinate()
  n104 = OneMinus(None=ComponentMask(None=@n103) {g=true})
  n105 = Multiply(
   A=SmoothStep(Max=ScalarParameter() {group=00- FlameShape, parameterName=FlameBaseSoftness}, Value=@n104) {constMax=0.2}
   B=SmoothStep(Max=ScalarParameter() {defaultValue=0.8, group=00- FlameShape, parameterName=FlameTipSoftness}, Value=OneMinus(None=@n104)) {constMax=0.1}
  )
  n106 = ComponentMask(None=@n103) {r=true}
  n107 = SmoothStep(
   Max=ScalarParameter() {defaultValue=0.43, group=00- FlameShape, parameterName=FlameSideSoftness}
   Value=Multiply(A=@n106, B=OneMinus(None=@n106))
  ) {constMax=0.5}
  n108 = NamedRerouteUsage()
  n109 = ScalarParameter() {defaultValue=0.55, group=00- FlameShape, parameterName=SubstractDisto}
  n110 = ComponentMask(None=Subtract(A=Multiply(A=@n105, B=@n107), B=Multiply(A=@n108, B=@n109) {constB=0.25})) {r=true}
  n111 = ScalarParameter() {defaultValue=4, group=00- FlameShape, parameterName=MainShapeMult}
  NamedRerouteDeclaration(None=Saturate(None=Multiply(A=@n110, B=@n111) {constB=3})) {name=MainShape, nodeColor=<rgba 0.2188 0 1 1>}
  n112 = ParticleColor()
  n113 = MaterialFunctionCall(Float2=TextureCoordinate()) {materialFunction=&BreakOutFloat2Components}
  n114 = MaterialFunctionCall(X=OneMinus(None=@n113.G), Y=@n113.R) {materialFunction=&MakeFloat2}
  n115 = Power(
   Base=OneMinus(None=ComponentMask(None=@n114.Result) {r=true})
   Exp=ScalarParameter() {defaultValue=8, group=03- Color, parameterName=VerticalColorMaskPower}
  )
  n116 = ScalarParameter() {defaultValue=3, group=03- Color, parameterName=VerticalColorMaskMult}
  n117 = ScalarParameter() {defaultValue=0.2, group=03- Color, parameterName=FirstColorThreshold}
  n118 = ScalarParameter() {defaultValue=1, group=03- Color, parameterName=FirstColorSoftness}
  n119 = NamedRerouteUsage()
  n120 = Saturate(None=SmoothStep(Min=@n117, Max=Add(A=@n117, B=@n118), Value=Saturate(None=@n119)) {constMin=0.5})
  n121 = ScalarParameter() {defaultValue=1.3, group=03- Color, parameterName=ColorNoiseAmount}
  n122 = Add(A=Saturate(None=Multiply(A=@n115, B=@n116)), B=Saturate(None=Multiply(A=@n120, B=@n121))) {constB=0}
  n123 = LinearInterpolate(
   A=VectorParameter() {group=03- Color, parameterName=ColorA, sortPriority=0, defaultValue=<rgba 0.007 0.7 0.0994 1>}.RGB
   B=VectorParameter() {group=03- Color, parameterName=ColorB, sortPriority=0, defaultValue=<rgba 0.3308 0.92 0 1>}.RGB
   Alpha=LinearInterpolate(B=ScalarParameter() {defaultValue=1, group=03- Color, parameterName=SecColorAmount}, Alpha=Saturate(None=@n122))
  )
  n124 = VectorParameter() {group=03- Color, parameterName=ColorC, defaultValue=<rgba 0 0 0 1>}
  n125 = NamedRerouteUsage()
  n126 = Saturate(None=Power(Base=OneMinus(None=@n125), Exp=ScalarParameter() {defaultValue=1, group=03- Color, parameterName=GlowMaskPower}))
  n127 = Multiply(
   A=Multiply(A=LinearInterpolate(A=@n123, B=@n124.RGB, Alpha=@n126), B=NamedRerouteUsage())
   B=ScalarParameter() {defaultValue=1, group=03- Color, parameterName=Emissive, sliderMax=2}
  )
  NamedRerouteDeclaration(None=MaterialFunctionCall(Emissive=Multiply(A=@n112.RGB, B=@n127) {constA=1}) {materialFunction=&MF_FX_EmissiveAdjustments}.Result) {name=Emissive, nodeColor=<rgba 0.2656 0 1 1>}
  NamedRerouteDeclaration(None=@n76.R) {name=ColorNoise, nodeColor=<rgba 0.3359 0 1 1>}
  NamedRerouteDeclaration(None=@n102.B) {name=FlameTrailInnerColor, nodeColor=<rgba 0 1 0.25 1>}
  NamedRerouteUsage()
  NamedRerouteUsage()
  NamedRerouteUsage()
  ScalarParameter() {defaultValue=1.5, group=WPO, parameterName=WPO Wiggle, sliderMax=10}
  VertexColor()
  VertexColor()
 ]
}
