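# Physics asset for the M_MED_HerbHutch_Breath male player body, covering a left-side pouch and the body parts around it.
# Capsules represent the pelvis, left thigh and two pouch bones.
# Constraints chain pouch_l to base_pouch_l to the pelvis, with the thigh also joined to the pelvis.
# The pouch base has free swing and twist on the pelvis, while the joint between the pouch bones has zero-degree angular limits.
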PhysicsAsset /BRCosmetics/Characters/Player/Male/Medium/Bodies/M_MED_HerbHutch_Breath/Meshes/M_MED_HerbHutch_Breath_Physics
bodies {
 base_pouch_l [{
   shapeType=Capsule
   length=10.1457
   radius=1.7127
   shapeName=base_pouch_l_capsule
   center=<0.5477 1.6321 -4.9433>
   rotation=<rot 84.6476 126.2812 -154.1153>
  }]
 pelvis [{shapeType=Capsule, length=4.0268, radius=13.5261, center=<-0.2991 -1.0473 1.401>, rotation=<rot 4.2788 -90.0002 -3.4694>}]
 pouch_l [{
   shapeType=Capsule
   length=10.1457
   radius=1.7127
   shapeName=base_pouch_l_capsule
   center=<0.7079 1.4139 -7.8325>
   rotation=<rot 84.6476 126.2812 -154.1152>
  }]
 thigh_l [{shapeType=Capsule, length=29.8186, radius=9.9065, center=<-18.7508 0.4327 0.2381>, rotation=<rot 0 -94.0931 -90.3614>}]
}
constraints [
 {bone1Name=thigh_l, bone2Name=pelvis, swing1LimitDegrees=45, swing2LimitDegrees=45, twistLimitDegrees=45}
 {
  bone1Name=base_pouch_l
  bone2Name=pelvis
  swing1LimitDegrees=10
  swing1Motion=Free
  swing2LimitDegrees=10
  swing2Motion=Free
  twistLimitDegrees=0
  twistMotion=Free
 }
 {bone1Name=pouch_l, bone2Name=base_pouch_l, swing1LimitDegrees=0, swing2LimitDegrees=0, twistLimitDegrees=0}
]
