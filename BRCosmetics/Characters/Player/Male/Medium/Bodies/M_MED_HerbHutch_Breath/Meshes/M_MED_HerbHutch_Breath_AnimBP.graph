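# Post-processing style AnimBP for the HerbHutch Breath male medium body cosmetic, adding secondary motion and fit fixes to the pose copied from the attached parent mesh.
# Control Rig and RigidBody feed cached neck and shoulder offsets, then Anim Dynamics springs armbands, zippers and scarf before an attach_backpack adjustment.
# EventGraph derives backpack offset variables from spine bend and flags whether the HerbHutch pack is equipped on customization load.
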AnimBlueprint /BRCosmetics/Characters/Player/Male/Medium/Bodies/M_MED_HerbHutch_Breath/Meshes/M_MED_HerbHutch_Breath_AnimBP
target_skeleton=&/M_MED_HerbHutch_Breath_Skeleton
graphs [
 AnimGraph {nodes [
   n1 = Anim Dynamics - Bone: dyn_armband_low(ComponentPose=Local To Component(LocalPose=Use cached pose 'FixCache'()), Alpha=1, ExternalForce="0,0,0") {
    angularDampingOverride=1
    angularSpringConstant=20000
    bAngularSpring=true
    bOverrideAngularDamping=true
    bUseGravityOverride=true
    gravityScale=0.35
    simulationSpace=RootRelative
    alphaBoolBlend {blendOption=Linear}
    boundBone {boneName=dyn_armband_low}
    componentLinearAccScale=<1 1 1>
    componentLinearVelScale=<1 1 1>
    gravityOverride {z=-980}
    physicsBodyDefinitions [{
      boundBone {boneName=dyn_armband_low}
      boxExtents=<15 60 60>
      constraintSetup {angularLimitsMax=<15 15 15>, angularLimitsMin=<-15 -15 -15>, angularTarget=<1 0 0>}
      localJointOffset=<3 0 0>
     }]
   }
   n2 = Anim Dynamics - Bone: dyn_armband_hi(ComponentPose=@n1, Alpha=1, ExternalForce="0,0,0") {
    angularDampingOverride=1
    angularSpringConstant=20000
    bAngularSpring=true
    bOverrideAngularDamping=true
    bUseGravityOverride=true
    gravityScale=0.35
    simulationSpace=RootRelative
    alphaBoolBlend {blendOption=Linear}
    boundBone {boneName=dyn_armband_hi}
    componentLinearAccScale=<1 1 1>
    componentLinearVelScale=<1 1 1>
    gravityOverride {z=-980}
    physicsBodyDefinitions [{
      boundBone {boneName=dyn_armband_hi}
      boxExtents=<15 60 60>
      constraintSetup {angularLimitsMax=<15 15 15>, angularLimitsMin=<-15 -15 -15>, angularTarget=<1 0 0>}
      localJointOffset=<3 0 0>
     }]
   }
   n3 = Anim Dynamics - Bone: zipper_02_r(ComponentPose=@n2, Alpha=1, ExternalForce="0,0,0") {
    angularDampingOverride=1
    angularSpringConstant=20000
    bAngularSpring=true
    bOverrideAngularDamping=true
    bUseGravityOverride=true
    gravityScale=0.35
    simulationSpace=RootRelative
    alphaBoolBlend {blendOption=Linear}
    boundBone {boneName=zipper_02_r}
    componentLinearAccScale=<1 1 1>
    componentLinearVelScale=<1 1 1>
    gravityOverride {z=-980}
    physicsBodyDefinitions [{
      boundBone {boneName=zipper_02_r}
      boxExtents=<20 30 30>
      constraintSetup {angularLimitsMax=<0 10 0>, angularLimitsMin=<0 -15 0>, angularTarget=<1 0 0>}
      localJointOffset=<6 0 0>
     }]
   }
   n4 = Anim Dynamics - Bone: zipper_01_r(ComponentPose=@n3, Alpha=1, ExternalForce="0,0,0") {
    angularDampingOverride=1
    angularSpringConstant=20000
    bAngularSpring=true
    bOverrideAngularDamping=true
    bUseGravityOverride=true
    gravityScale=0.35
    simulationSpace=RootRelative
    alphaBoolBlend {blendOption=Linear}
    boundBone {boneName=zipper_01_r}
    componentLinearAccScale=<1 1 1>
    componentLinearVelScale=<1 1 1>
    gravityOverride {z=-980}
    physicsBodyDefinitions [{
      boundBone {boneName=zipper_01_r}
      boxExtents=<20 30 30>
      constraintSetup {angularLimitsMax=<0 10 0>, angularLimitsMin=<0 -15 0>, angularTarget=<1 0 0>}
      localJointOffset=<6 0 0>
     }]
   }
   n5 = Anim Dynamics - Bone: zipper_02_l(ComponentPose=@n4, Alpha=1, ExternalForce="0,0,0") {
    angularDampingOverride=1
    angularSpringConstant=20000
    bAngularSpring=true
    bOverrideAngularDamping=true
    bUseGravityOverride=true
    gravityScale=0.35
    simulationSpace=RootRelative
    alphaBoolBlend {blendOption=Linear}
    boundBone {boneName=zipper_02_l}
    componentLinearAccScale=<1 1 1>
    componentLinearVelScale=<1 1 1>
    gravityOverride {z=-980}
    physicsBodyDefinitions [{
      boundBone {boneName=zipper_02_l}
      boxExtents=<20 30 30>
      constraintSetup {angularLimitsMax=<0 10 0>, angularLimitsMin=<0 -15 0>, angularTarget=<1 0 0>}
      localJointOffset=<6 0 0>
     }]
   }
   n6 = Anim Dynamics - Bone: zipper_01_l(ComponentPose=@n5, Alpha=1, ExternalForce="0,0,0") {
    angularDampingOverride=1
    angularSpringConstant=20000
    bAngularSpring=true
    bOverrideAngularDamping=true
    bUseGravityOverride=true
    gravityScale=0.35
    simulationSpace=RootRelative
    alphaBoolBlend {blendOption=Linear}
    boundBone {boneName=zipper_01_l}
    componentLinearAccScale=<1 1 1>
    componentLinearVelScale=<1 1 1>
    gravityOverride {z=-980}
    physicsBodyDefinitions [{
      boundBone {boneName=zipper_01_l}
      boxExtents=<20 30 30>
      constraintSetup {angularLimitsMax=<0 10 0>, angularLimitsMin=<0 -15 0>, angularTarget=<1 0 0>}
      localJointOffset=<6 0 0>
     }]
   }
   n7 = Anim Dynamics - Bone: scarf_front_c(ComponentPose=@n6, Alpha=1, ExternalForce="0,0,0") {
    angularDampingOverride=1
    angularSpringConstant=20000
    bAngularSpring=true
    bOverrideAngularDamping=true
    bUseGravityOverride=true
    gravityScale=0.35
    simulationSpace=RootRelative
    alphaBoolBlend {blendOption=Linear}
    boundBone {boneName=scarf_front_c}
    componentLinearAccScale=<1 1 1>
    componentLinearVelScale=<1 1 1>
    gravityOverride {z=-980}
    physicsBodyDefinitions [{
      boundBone {boneName=scarf_front_c}
      boxExtents=<20 20 30>
      constraintSetup {angularLimitsMax=<0 -7 0>, angularLimitsMin=<0 -15 0>, angularTarget=<1 0 0>}
      localJointOffset=<4 0 0>
     }]
   }
   n8 = Anim Dynamics - Bone: scarf_c(ComponentPose=@n7, Alpha=1, ExternalForce="0,0,0") {
    angularDampingOverride=1
    angularSpringConstant=20000
    bAngularSpring=true
    bOverrideAngularDamping=true
    bUseGravityOverride=true
    gravityScale=0.35
    simulationSpace=RootRelative
    alphaBoolBlend {blendOption=Linear}
    boundBone {boneName=scarf_c}
    componentLinearAccScale=<1 1 1>
    componentLinearVelScale=<1 1 1>
    gravityOverride {z=-980}
    physicsBodyDefinitions [{
      boundBone {boneName=scarf_c}
      boxExtents=<20 50 30>
      constraintSetup {angularLimitsMax=<0 0 0>, angularLimitsMin=<0 -20 0>, angularTarget=<1 0 0>}
      localJointOffset=<4 0 0>
     }]
   }
   n9 = "Transform (Modify) Bone - Bone: attach_backpack"!(ComponentPose=@n8, bAlphaBoolEnabled:bool=True, Translation="0,0,0", Rotation="0,0,0") -> (Pose) {
    alphaInputType=Bool
    rotationMode=BMM_Additive
    rotationSpace=BCS_BoneSpace
    translationMode=BMM_Additive
    translationSpace=BCS_BoneSpace
    alphaBoolBlend {blendInTime=0.1, blendOption=Linear, blendOutTime=0.1}
    boneToModify {boneName=attach_backpack}
   }
   Output Pose(Result=Component To Local(ComponentPose=@n9)) {layerGroup=None}
   n10 = Control Rig(Source=Copy Pose From Mesh() {bUseAttachedParent=true}) {
    alphaBoolBlend {blendOption=Linear}
    controlRigAssetReference {blueprintRigClass=&/M_MED_HerbHutch_Breath_CtrlRig.M_MED_HerbHutch_Breath_CtrlRig_C}
   }
   n11 = "Transform (Modify) Bone - Bone: base_pouch_l"!(ComponentPose=Local To Component(LocalPose=@n10), Alpha=1, Translation="0.5,0,0") -> (Pose) {
    rotationMode=BMM_Additive
    rotationSpace=BCS_BoneSpace
    translationMode=BMM_Additive
    translationSpace=BCS_BoneSpace
    alphaBoolBlend {blendOption=Linear}
    boneToModify {boneName=base_pouch_l}
    translation {x=0.5}
   }
   n12 = RigidBody(ComponentPose=@n11) {
    bClampLinearTranslationLimitToRefPose=true
    overridePhysicsAsset=&/M_MED_HerbHutch_Breath_Physics
    simulationSpace=BaseBoneSpace
    alphaBoolBlend {blendOption=Linear}
    baseBoneRef {boneName=pelvis}
   }
   Save cached pose 'ControlRigPhysicsCache'(Pose=Component To Local(ComponentPose=@n12)) {cacheName=ControlRigPhysicsCache}
   n13 = "Transform (Modify) Bone - Bone: neck_02_front"(
    ComponentPose=Local To Component(LocalPose=Use cached pose 'ControlRigPhysicsCache'())
    Alpha=1
    Translation="0,1,0"
    Rotation="0,0,0"
    Scale="1,1,1"
   ) {
    translationMode=BMM_Additive
    translationSpace=BCS_BoneSpace
    alphaBoolBlend {blendOption=Linear}
    boneToModify {boneName=neck_02_front}
    translation {y=1}
   }
   n14 = "Transform (Modify) Bone - Bone: shoulders_r"!(ComponentPose=@n13, bAlphaBoolEnabled:bool=True, Translation="0.5,-1,0.687", Rotation="-5,5,0") -> (Pose) {
    alphaInputType=Bool
    rotationMode=BMM_Additive
    rotationSpace=BCS_BoneSpace
    translationMode=BMM_Additive
    translationSpace=BCS_BoneSpace
    alphaBoolBlend {blendInTime=1, blendOption=Linear, blendOutTime=1}
    boneToModify {boneName=shoulders_r}
    rotation {pitch=-5, yaw=5}
    translation=<0.5 -1 0.687>
   }
   n15 = "Transform (Modify) Bone - Bone: shoulders_l"!(ComponentPose=@n14, bAlphaBoolEnabled:bool=True, Translation="0,0,-1", Rotation="0,0,0") -> (Pose) {
    alphaInputType=Bool
    rotationMode=BMM_Additive
    rotationSpace=BCS_BoneSpace
    translationMode=BMM_Additive
    translationSpace=BCS_BoneSpace
    alphaBoolBlend {blendInTime=1, blendOption=Linear, blendOutTime=1}
    boneToModify {boneName=shoulders_l}
    translation {z=-1}
   }
   n16 = "Transform (Modify) Bone - Bone: shoulders_r"(ComponentPose=@n15, bAlphaBoolEnabled=True, Translation="0,0,0", Rotation="0,-5,10") {
    alphaInputType=Bool
    rotationMode=BMM_Additive
    rotationSpace=BCS_BoneSpace
    translationMode=BMM_Additive
    translationSpace=BCS_BoneSpace
    alphaBoolBlend {blendInTime=1, blendOption=Linear, blendOutTime=1}
    boneToModify {boneName=shoulders_r}
    rotation {roll=10, yaw=-5}
   }
   n17 = "Transform (Modify) Bone - Bone: shoulders_l"(ComponentPose=@n16, bAlphaBoolEnabled=True, Translation="0,0,0", Rotation="-0.080,-5.021,-5.000") {
    alphaInputType=Bool
    rotationMode=BMM_Additive
    rotationSpace=BCS_BoneSpace
    translationMode=BMM_Additive
    translationSpace=BCS_BoneSpace
    alphaBoolBlend {blendInTime=1, blendOption=Linear, blendOutTime=1}
    boneToModify {boneName=shoulders_l}
    rotation=<rot -0.08 -5.0211 -5>
   }
   Save cached pose 'FixCache'(Pose=Component To Local(ComponentPose=@n17)) {cacheName=FixCache}
  ]}
 EventGraph {nodes [
   Event BlueprintUpdateAnimation()
   n1 = GetOwnerSkeletalMeshComponent()
   n2 = BreakRotator(InRot=BreakTransform(InTransform=GetDeltaTransformFromRefPose(self=@n1, BoneName=spine_05, BaseName=spine_01)).Rotation)
   > Set backpack_rotate(backpack_rotate=MakeRotator(Roll=MapRangeClamped(Value=@n2.Yaw, InRangeA=0, InRangeB=-30, OutRangeB=20)))
   > Set backpack_translation(backpack_translation=MakeVector(Y=MapRangeClamped(Value=@n2.Yaw, InRangeA=0, InRangeB=-40, OutRangeB=-5), Z=-1.5))
   Event OnCharacterCustomizationLoaded()
   > n3 = Cast To FortPlayerPawn(Object=TryGetPawnOwner())
   n4 = GetSkeletalMeshForPartType(self=@n3, PartType=Backpack)
   > Set isHerbHutchBP(isHerbHutchBP:bool=MatchesWildcard(SourceString=GetDisplayName(Object=@n4), Wildcard=*M_MED_HerbHutch_Pack*, SearchCase=CaseSensitive)) -> (Output_Get:bool)
  ]}
]
variables [
 backpack_rotate {type=Rotator}
 backpack_translation {type=Vector}
 isHerbHutchBP {type=boolean}
]
